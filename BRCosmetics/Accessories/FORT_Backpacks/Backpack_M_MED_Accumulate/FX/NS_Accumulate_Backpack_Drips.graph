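# Niagara mesh-particle system for the Accumulate backpack cosmetic, producing swinging drips and falling drops anchored to its skeletal mesh.
# Hero_Drips (burst, looping lifetime) and Small_Drips (continuous) use spring-driven pendulums pivoted on sampled bones, with spawn groups flipping offset and rest axis.
# They pass pivot, position and rigidity to the material through dynamic parameters while curves stretch the mesh.
# Drops spawn in a ring, speed up vertically and are suppressed in front-end preview.
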NiagaraSystem /BRCosmetics/Accessories/FORT_Backpacks/Backpack_M_MED_Accumulate/FX/NS_Accumulate_Backpack_Drips
dependencies {usedDataInterfaces [&NiagaraDataInterfaceCurve], usedDynamicInputs [], usedModules [&SystemState]}
emitters {
 Drops {
  emitterUpdateScript [
   &EmitterState()
   &SpawnRate(
    Spawn Probability=&MakeCustomFloatFromBool(Bool=$User.bIsFrontEndPreview, True Float=0, False Float=1)
    SpawnRate=8
    Use Spawn Probability=true
   )
  ]
  particleSpawnScript [
   /Niagara/Modules/Spawn/Initialization/InitializeParticle(
    Lifetime=&UniformRangedFloat(Minimum=0.85, Maximum=1.2)
    Mesh Scale=&VectorFromFloat(Value=&RandomRangeFloat(Minimum=0.0075, Maximum=0.02))
    Module.Write Color=true
    Module.Write Lifetime=true
    Module.Write Mass=true
    Module.Write Material Random=true
    Module.Write Position=false
    Module.Write Scale=true
    Module.Write SpriteRotation=false
    Module.Write SpriteSize=true
   )
   &SkeletalMeshLocation(
    Bone Sampling Mode=::NewEnumerator4
    Inherit Velocity=false
    MaskSpawnGroup=false
    Sampled Position Offset=<0 -4 -5>
    Sampled Position Offset Coordinate Space:ENiagaraMeshTransforms=::NewEnumerator3
    Skeletal Mesh=$User.SkeletalMesh
   )
   &CylinderLocation(
    Cylinder Height=0
    Cylinder Radius=8
    MaskSpawnGroup=false
    Module.Enable Lathe Profile=false
    Module.Override Randomness=false
    Module.Surface Only=true
    Orientation Axis:ENiagaraOrientationAxis=::YAxis
    Surface Only Band Thickness=0.5
    Use Endcaps In Surface Only Mode=false
   )
   &AddVelocity()
  ]
  particleUpdateScript [
   &ParticleState()
   &ScaleVelocity(Velocity Scale=&MakeVector(
     X=1
     Y=1
     Z=&FloatFromCurve(FloatCurve=DataInterface() {propertyValues {Curve {keys [
          {arriveTangent=7.4745, interpMode=RCIM_Cubic, leaveTangent=7.4745, tangentMode=RCTM_User, time=0, value=0}
          {arriveTangent=14.58, interpMode=RCIM_Cubic, leaveTangent=14.58, tangentMode=RCTM_User, time=1, value=4}
         ]}}})
    ))
   &InheritSourceMovement(
    Applied Position Scale=&Multiply_VectorByFloat(
     Vector=<1 1 1>
     Float=&FloatFromCurve(FloatCurve=DataInterface() {propertyValues {Curve {keys [<at 0.4 1>,<at 1 0>]}}})
    )
    ApplyForce=false
    ApplyPosition=true
    ApplyPositionFalloff=false
   )
   &SolveForcesAndVelocity(Clamp Velocity=false, Limit Acceleration=false)
   &ScaleMeshSize(Scale Factor=&VectorFromFloat(Value=&FloatFromCurve(FloatCurve=DataInterface() {propertyValues {Curve {keys [
          {arriveTangent=5.6263, interpMode=RCIM_Cubic, leaveTangent=5.6263, tangentMode=RCTM_User, time=0.2, value=0}
          {interpMode=RCIM_Cubic, tangentMode=RCTM_User, time=0.6, value=1.5}
          {arriveTangent=-4.9972, interpMode=RCIM_Cubic, leaveTangent=-4.9972, tangentMode=RCTM_User, time=0.9, value=0}
         ]}}})))
  ]
  rendererClasses [&NiagaraMeshRendererProperties]
  renderers [{rendererClass=&NiagaraMeshRendererProperties, rendererIndex=0}]
 }
 Hero_Drips {
  emitterUpdateScript [
   &EmitterState()
   &SpawnBurst_Instantaneous(Spawn Count=2, Spawn Probability=HlslExpression() {hlslExpression="Emitter.LoopCount == 0"}, Use Spawn Probability=true)
  ]
  particleSpawnScript [
   &FN_SetSpawnGroup(Number of Spawn Groups=2)
   /SetVariables(
    Particles.PendulumDrag=1
    Particles.PendulumLengthScale:NiagaraFloat=&RandomRangeFloat(Minimum=0.3, Maximum=0.4)
    Particles.PendulumMaskFalloff=1.5
    Particles.PendulumMeshWidth:NiagaraFloat=&RandomRangeFloat(Minimum=0.09, Maximum=0.11)
    Particles.PendulumRigidity=0.05
   ) {isSetParameters=true, module=SetVariables_#1}
   &InitializeParticle(
    Lifetime=&UniformRangedFloat(Minimum=5.5, Maximum=9)
    Mesh Scale=&MakeVector(X=$Particles.PendulumMeshWidth, Y=$Particles.PendulumMeshWidth, Z=$Particles.PendulumLengthScale)
    Module.Write Color=true
    Module.Write Lifetime=true
    Module.Write Mass=true
    Module.Write Material Random=true
    Module.Write Position=false
    Module.Write Scale=true
    Module.Write SpriteRotation=false
    Module.Write SpriteSize=true
   )
   &CameraOffset(Camera Offset Amount=2)
  ]
  particleUpdateScript [
   &ParticleState(
    Kill Particles When Lifetime Has Elapsed=false
    Let Infinitely Lived Particles Die When Emitter Deactivates=true
    Loop Particles Lifetime=true
   )
   &SkeletalMeshLocation(
    Apply To Particle Position=false
    Bone Sampling Mode=::NewEnumerator4
    Inherit Velocity=false
    MaskSpawnGroup=false
    Skeletal Mesh=$User.SkeletalMesh
   )
   /SetVariables(Particles.MeshOrigin:Vector3f=&Add_Vector(
     A=$Output.SkeletalMeshLocation.SampledPosition
     B=&RotateVectorByQuaternion(
      Vector To Rotate=&MaskVectorBySpawnGroup(Masked Vector=<-4 -3.85 -1>, Passthrough Vector=<1.35 -4.25 -10>)
      Quaternion=$Output.SkeletalMeshLocation.SampledRotation
     )
    )) {isSetParameters=true, module=SetVariables_#2}
   &Drag(Drag=$Particles.PendulumDrag, Use Linear Drag=true, Use Rotational Drag=false)
   &GravityForce()
   &PendulumSetup(
    Module.Enable Spring Driver=true
    Pendulum Length:NiagaraFloat=&Multiply_Float(A=100, B=$Particles.PendulumLengthScale)
    Pendulum Pivot:Vector3f=$Particles.MeshOrigin
    Pendulum Rest Axis:Vector3f=&MaskVectorBySpawnGroup(Masked Vector=<0 0 1>, Passthrough Vector=<0 0 -1>)
    "Spring Driven Constraint (Tightness)"=10
   )
   /SetVariables(Particles.MeshFacing:Vector3f=&NormalizeVector(Vector To Normalize=&Subtract_Vector(A=$Particles.MeshOrigin, B=Linked() {linkedVariable {name=Particles.Position, type=NiagaraPosition}}))) {isSetParameters=true, module=SetVariables_#3}
   &self:Pendulum_Facing(
    Bone Quat:Quat4f=$Output.SkeletalMeshLocation.SampledRotation
    FromVector:Vector3f=&MaskVectorBySpawnGroup(Masked Vector=<0 0 -1>, Passthrough Vector=<0 0 1>)
    ToVector:Vector3f=$Particles.MeshFacing
    X Vector:Vector3f=&MaskVectorBySpawnGroup(Masked Vector=<-1 0 0>, Passthrough Vector=<1 0 0>)
    Y Vector:Vector3f=<0 0 0>
   ) {module=Pendulum_Facing}
   &SolveForcesAndVelocity(Clamp Velocity=false, Limit Acceleration=false)
   &PendulumConstraint()
   &CalculateAccurateVelocity()
   &DynamicMaterialParameters(
    Index 1 Param 3=1
    Index 1 Param 4=&FloatFromCurve(FloatCurve=DataInterface() {propertyValues {Curve {keys [
         {arriveTangent=0.1436, interpMode=RCIM_Cubic, leaveTangent=0.1436, tangentMode=RCTM_User, time=0, value=-0.2}
         {interpMode=RCIM_Cubic, leaveTangent=2.3573, time=1, value=3}
        ]}}})
    Index 2 Param 1=HlslExpression() {hlslExpression=Particles.MeshOrigin.x}
    Index 2 Param 2=HlslExpression() {hlslExpression=Particles.MeshOrigin.y}
    Index 2 Param 3=HlslExpression() {hlslExpression=Particles.MeshOrigin.z}
    Index 2 Param 4=$Particles.PendulumMaskFalloff
    Index 3 Param 1=&MakeFloatFromVector(Vector=Linked() {linkedVariable {name=Particles.Position, type=NiagaraPosition}}, Channel=::NewEnumerator0)
    Index 3 Param 2=&MakeFloatFromVector(Vector=Linked() {linkedVariable {name=Particles.Position, type=NiagaraPosition}}, Channel=::NewEnumerator1)
    Index 3 Param 3=&MakeFloatFromVector(Vector=Linked() {linkedVariable {name=Particles.Position, type=NiagaraPosition}}, Channel=::NewEnumerator2)
    Index 3 Param 4=$Particles.PendulumRigidity
    Module.Param0WriteEnabled1=true
    Module.Param0WriteEnabled2=true
    Module.Param0WriteEnabled3=true
    Module.Param1WriteEnabled1=true
    Module.Param1WriteEnabled2=true
    Module.Param1WriteEnabled3=true
    Module.Param2WriteEnabled1=true
    Module.Param2WriteEnabled2=true
    Module.Param2WriteEnabled3=true
    Module.Param3WriteEnabled1=true
    Module.Param3WriteEnabled2=true
    Module.Param3WriteEnabled3=true
    Write Parameter Index 0=false
    Write Parameter Index 1=true
    Write Parameter Index 2=true
    Write Parameter Index 3=true
   )
   &ScaleMeshSize(Scale Factor=&MakeVector(
     X=1
     Y=1
     Z=&FloatFromCurve(FloatCurve=DataInterface() {propertyValues {Curve {keys [
          {interpMode=RCIM_Cubic, tangentMode=RCTM_User, time=0, value=0.6}
          {interpMode=RCIM_Cubic, tangentMode=RCTM_User, time=0.6, value=1.35}
          {tangentMode=RCTM_User, time=1, value=0.6}
         ]}}})
    ))
   &ScaleMeshSize(Scale Factor=&VectorFromFloat(Value=&MaskFloatBySpawnGroup(Masked Float=1, Passthrough Float=0))) {module=ScaleMeshSize001}
  ]
  rendererClasses [&NiagaraMeshRendererProperties]
  renderers [{rendererClass=&NiagaraMeshRendererProperties, rendererIndex=0}]
 }
 Small_Drips {
  emitterUpdateScript [
   &EmitterState()
   &SpawnRate(Spawn Probability=0.75, Use Spawn Probability=true)
  ]
  particleSpawnScript [
   &FN_SetSpawnGroup(Number of Spawn Groups=2)
   /SetVariables(
    Particles.PendulumDrag=1
    Particles.PendulumLengthScale:NiagaraFloat=&RandomRangeFloat(Minimum=0.2, Maximum=0.25)
    Particles.PendulumMaskFalloff=0.5
    Particles.PendulumMeshWidth:NiagaraFloat=&RandomRangeFloat(Minimum=0.05, Maximum=0.1)
    Particles.PendulumRigidity=0.06
   ) {isSetParameters=true, module=SetVariables_#1}
   &InitializeParticle(
    Lifetime=&UniformRangedFloat(Minimum=5, Maximum=6)
    Mesh Scale=&MakeVector(X=$Particles.PendulumMeshWidth, Y=$Particles.PendulumMeshWidth, Z=$Particles.PendulumLengthScale)
    Module.Write Color=true
    Module.Write Lifetime=true
    Module.Write Mass=true
    Module.Write Material Random=true
    Module.Write Position=false
    Module.Write Scale=true
    Module.Write SpriteRotation=false
    Module.Write SpriteSize=true
   )
  ]
  particleUpdateScript [
   &ParticleState()
   &SkeletalMeshLocation(
    Apply To Particle Position=false
    Bone Sampling Mode=::NewEnumerator4
    Inherit Velocity=false
    MaskSpawnGroup=false
    Skeletal Mesh=$User.SkeletalMesh
   )
   /SetVariables(Particles.MeshOrigin:Vector3f=&Add_Vector(
     A=$Output.SkeletalMeshLocation.SampledPosition
     B=&RotateVectorByQuaternion(
      Vector To Rotate=&MaskVectorBySpawnGroup(
       Masked Vector=&RandomRangeVector(Minimum=<-5 -3.75 -1>, Maximum=<5 -4.25 0>)
       Passthrough Vector=&RandomRangeVector(Minimum=<-5 -3.75 -10>, Maximum=<5 -4.25 -12>)
      )
      Quaternion=$Output.SkeletalMeshLocation.SampledRotation
     )
    )) {isSetParameters=true, module=SetVariables_#2}
   &Drag(Drag=$Particles.PendulumDrag, Use Linear Drag=true, Use Rotational Drag=false)
   &GravityForce()
   &PendulumSetup(
    Module.Enable Spring Driver=true
    Pendulum Length:NiagaraFloat=&Multiply_Float(A=100, B=$Particles.PendulumLengthScale)
    Pendulum Pivot:Vector3f=$Particles.MeshOrigin
    Pendulum Rest Axis:Vector3f=&MaskVectorBySpawnGroup(Masked Vector=<0 0 1>, Passthrough Vector=<0 0 -1>)
    "Spring Driven Constraint (Tightness)"=10
   )
   /SetVariables(Particles.MeshFacing:Vector3f=&NormalizeVector(Vector To Normalize=&Subtract_Vector(A=$Particles.MeshOrigin, B=Linked() {linkedVariable {name=Particles.Position, type=NiagaraPosition}}))) {isSetParameters=true, module=SetVariables_#3}
   &self:Pendulum_Facing(
    Bone Quat:Quat4f=$Output.SkeletalMeshLocation.SampledRotation
    FromVector:Vector3f=&MaskVectorBySpawnGroup(Masked Vector=<0 0 -1>, Passthrough Vector=<0 0 1>)
    ToVector:Vector3f=$Particles.MeshFacing
    X Vector:Vector3f=&MaskVectorBySpawnGroup(Masked Vector=<-1 0 0>, Passthrough Vector=<1 0 0>)
    Y Vector:Vector3f=<0 0 0>
   ) {module=Pendulum_Facing}
   &SolveForcesAndVelocity(Clamp Velocity=false, Limit Acceleration=false)
   &PendulumConstraint()
   &CalculateAccurateVelocity()
   &DynamicMaterialParameters(
    Index 1 Param 3=1
    Index 1 Param 4=&FloatFromCurve(FloatCurve=DataInterface() {propertyValues {Curve {keys [
         {arriveTangent=3.6645, interpMode=RCIM_Cubic, leaveTangent=3.6645, tangentMode=RCTM_User, time=0.1, value=-0.1}
         {interpMode=RCIM_Cubic, leaveTangent=2.3573, time=0.9, value=2}
        ]}}})
    Index 2 Param 1=HlslExpression() {hlslExpression=Particles.MeshOrigin.x}
    Index 2 Param 2=HlslExpression() {hlslExpression=Particles.MeshOrigin.y}
    Index 2 Param 3=HlslExpression() {hlslExpression=Particles.MeshOrigin.z}
    Index 2 Param 4=$Particles.PendulumMaskFalloff
    Index 3 Param 1=&MakeFloatFromVector(Vector=Linked() {linkedVariable {name=Particles.Position, type=NiagaraPosition}}, Channel=::NewEnumerator0)
    Index 3 Param 2=&MakeFloatFromVector(Vector=Linked() {linkedVariable {name=Particles.Position, type=NiagaraPosition}}, Channel=::NewEnumerator1)
    Index 3 Param 3=&MakeFloatFromVector(Vector=Linked() {linkedVariable {name=Particles.Position, type=NiagaraPosition}}, Channel=::NewEnumerator2)
    Index 3 Param 4=$Particles.PendulumRigidity
    Module.Param0WriteEnabled1=true
    Module.Param0WriteEnabled2=true
    Module.Param0WriteEnabled3=true
    Module.Param1WriteEnabled1=true
    Module.Param1WriteEnabled2=true
    Module.Param1WriteEnabled3=true
    Module.Param2WriteEnabled1=true
    Module.Param2WriteEnabled2=true
    Module.Param2WriteEnabled3=true
    Module.Param3WriteEnabled1=true
    Module.Param3WriteEnabled2=true
    Module.Param3WriteEnabled3=true
    Write Parameter Index 0=false
    Write Parameter Index 1=true
    Write Parameter Index 2=true
    Write Parameter Index 3=true
   )
   &ScaleMeshSize(Scale Factor=&MakeVector(
     X=&FloatFromCurve(FloatCurve=DataInterface() {propertyValues {Curve {keys [
          <at 0 0 RCIM_Cubic>
          {arriveTangent=2.5, interpMode=RCIM_Cubic, leaveTangent=2.5, time=0.05, value=0.9}
          {arriveTangent=-0.7353, interpMode=RCIM_Cubic, leaveTangent=-0.7353, time=0.6, value=0.65}
          {arriveTangent=-4.2857, interpMode=RCIM_Cubic, leaveTangent=-4.2857, time=0.9, value=0.375}
          {interpMode=RCIM_Cubic, leaveTangent=-0.6579, time=0.95, value=0}
         ]}}})
     Y=&FloatFromCurve(FloatCurve=DataInterface() {propertyValues {Curve {keys [
          <at 0 0 RCIM_Cubic>
          {arriveTangent=2.5, interpMode=RCIM_Cubic, leaveTangent=2.5, time=0.05, value=0.9}
          {arriveTangent=-0.7353, interpMode=RCIM_Cubic, leaveTangent=-0.7353, time=0.6, value=0.65}
          {arriveTangent=-4.2857, interpMode=RCIM_Cubic, leaveTangent=-4.2857, time=0.9, value=0.375}
          {interpMode=RCIM_Cubic, leaveTangent=-0.6579, time=0.95, value=0}
         ]}}})
     Z=&FloatFromCurve(FloatCurve=DataInterface() {propertyValues {Curve {keys [
          {arriveTangent=1.8796, interpMode=RCIM_Cubic, time=0, value=0}
          {arriveTangent=-4.7929, interpMode=RCIM_Cubic, leaveTangent=-4.7929, tangentMode=RCTM_User, time=0.625, value=0.85}
          {arriveTangent=-1.7696, interpMode=RCIM_Cubic, leaveTangent=-1.7696, tangentMode=RCTM_User, time=0.95, value=0}
         ]}}})
    ))
  ]
  rendererClasses [&NiagaraMeshRendererProperties]
  renderers [{rendererClass=&NiagaraMeshRendererProperties, rendererIndex=0}]
 }
}
summary {
 systemName=NS_Accumulate_Backpack_Drips
 emitters [
  {emitterName=Small_Drips, rendererClasses [&NiagaraMeshRendererProperties]}
  {emitterName=Hero_Drips, rendererClasses [&NiagaraMeshRendererProperties]}
  {emitterName=Drops, rendererClasses [&NiagaraMeshRendererProperties]}
 ]
 userVariables [
  User.bIsFrontEndPreview {type=NiagaraBool}
  User.RibbonIDOffset {type=NiagaraInt32}
  User.SkeletalMesh {
   type=NiagaraDataInterfaceSkeletalMesh
   defaultValue {
    struct=NiagaraExt_VariableValue_DataInterface
    value {dataInterface=&self:NiagaraDataInterfaceSkeletalMesh_0}
   }
  }
 ]
}
