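# Skinned Accumulate backpack material: an emissive animated face on a black, rough surface, wrapped in shared character PreFX/PostFX functions.
# The IsBasePart switch selects either noisy Fresnel rim plus face, colored through curve-atlas rows, or a plain two-color Fresnel gradient.
# The face mask uses noise-distorted, bump-offset UVs and a curve-driven blink.
# AC_KillTrigger blends to a stretchable laugh face with squiggly polar streaks, moved by AC_Streaks.
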Material /BRCosmetics/Accessories/FORT_Backpacks/Backpack_M_MED_Accumulate/Materials/M_MED_Accumulate_Backpack
graph {
 properties {
  bUseMaterialAttributes=true
  bUsedWithMeshDeformer=true
  bUsedWithSkeletalMesh=true
  pixelDepthOffsetMode=PDOM_Legacy
  usedShadingModels=Default Lit
 }
 nodes [
  n1 = ScalarParameter() {defaultValue=1, parameterName=Exponent_02}
  n2 = TextureCoordinate()
  n3 = ScalarParameter() {defaultValue=1, parameterName=NoiseTiling_X}
  n4 = ScalarParameter() {defaultValue=1, parameterName=NoiseTiling_Y}
  n5 = TextureSample(UVs=Multiply(A=Panner(Coordinate=@n2) {speedY=0.25}, B=AppendVector(A=@n3, B=@n4))) {texture=/Game/Characters/Player/Male/Medium/Bodies/M_MED_Viper/Textures/Glitch/T_SoftNoise}
  n6 = Subtract(
   A=Power(Base=Fresnel(ExponentIn=@n1)) {constExponent=1}
   B=Multiply(A=ComponentMask(None=@n5.RGB) {r=true}, B=ScalarParameter() {defaultValue=0.25, parameterName=Noise_Strength})
  )
  Add(B=@n6)
  n7 = Multiply(A=Time(), B=ScalarParameter() {defaultValue=0.1, parameterName=BlinkFrequency})
  n8 = MaterialFunctionCall(Float=@n7) {materialFunction=&MF_Hash}
  Clamp(None=OneMinus(None=Power(Base=@n8.Float)))
  n9 = MaterialFunctionCall() {materialFunction=&BaseFN_CharacterFunction_PreFX}
  n10 = CurveAtlasRowParameter(CurveTime=Clamp(None=@n6)) {
   atlas=&AthenaCurveAtlas01
   curve=/Game/Characters/Player/Male/Medium/Bodies/M_MED_Accumulate/Materials/AccumulateBlackColorCurve
   defaultValue=276
   parameterName=RimColorCurve
  }
  n11 = Multiply(A=@n10, B=ScalarParameter() {defaultValue=0.5, parameterName=Rim_Strength})
  n12 = ScalarParameter() {parameterName=AC_KillTrigger}
  n13 = MaterialFunctionCall!(Coordinate=Add(A=TextureCoordinate(), B=Constant2Vector() {g=-0.05})) {materialFunction=/Game/Effects/Niagara/MasterMaterials/Functions/UVs/PolarCoordinates}
  n14 = ComponentMask(None=@n13.Result) {g=true, r=true}
  n15 = ComponentMask(None=@n14) {g=true}
  n16 = ScalarParameter() {parameterName=AC_Streaks}
  n17 = Multiply(A=@n16) {constB=-1}
  n18 = Frac(None=Multiply(A=Add(A=@n15, B=@n17), B=ScalarParameter() {defaultValue=2, parameterName=Bands}))
  n19 = ScalarParameter() {defaultValue=0.5, parameterName=BandSamplePos}
  n20 = ScalarParameter() {defaultValue=0.25, parameterName=BandRadius}
  n21 = ScalarParameter() {defaultValue=0.5, parameterName=BandHardness}
  n22 = ComponentMask(None=@n14) {r=true}
  n23 = ComponentMask(None=@n13.Result) {g=true}
  n24 = Add(A=@n23, B=@n16)
  n25 = ScalarParameter() {defaultValue=12, parameterName=DistortBands}
  n26 = SphereMask(
   A=Frac(None=Multiply(A=@n24, B=@n25))
   B=ScalarParameter() {defaultValue=0.5, parameterName=DisortBandSamplePos}
   Radius=ScalarParameter() {defaultValue=0.5, parameterName=DistortBandRadius}
   Hardness=ScalarParameter() {parameterName=DistortBandHardness}
  )
  n27 = Multiply(A=Subtract(A=Multiply(A=@n26) {constB=2}), B=ScalarParameter() {defaultValue=0.025, parameterName=SquiggleStrength})
  n28 = Frac(None=Multiply(A=Add(A=@n22, B=@n27), B=ScalarParameter() {defaultValue=2.5, parameterName=Streaks}))
  n29 = ScalarParameter() {defaultValue=0.5, parameterName=SamplePos}
  n30 = ScalarParameter() {defaultValue=0.1, parameterName=Radius}
  n31 = ScalarParameter() {parameterName=Hardness}
  n32 = Multiply(A=@n23, B=Multiply(A=SphereMask(A=@n18, B=@n19, Radius=@n20, Hardness=@n21), B=SphereMask(A=@n28, B=@n29, Radius=@n30, Hardness=@n31)))
  n33 = Multiply(
   A=LinearInterpolate(
    A=VectorParameter() {parameterName=Rim_Color_02, defaultValue=<rgba 0 0.4993 0.535 1>}.RGB
    B=VectorParameter() {parameterName=Rim_Color_01, defaultValue=<rgba 1 1 1 1>}.RGB
    Alpha=@n32
   )
   B=@n32
  )
  n34 = LinearInterpolate(
   A=ScalarParameter() {defaultValue=5, parameterName=SquiggleBrightnessMin}
   B=ScalarParameter() {defaultValue=50, parameterName=SquiggleBrightness}
   Alpha=@n23
  )
  n35 = ComponentMask(None=TextureCoordinate()) {g=true}
  n36 = CurveAtlasRowParameter(CurveTime=Frac(None=@n7)) {
   atlas=&AthenaCurveAtlas01
   curve=/Game/Characters/Player/Male/Medium/Bodies/M_MED_Accumulate/Materials/AccumulateSpriteAnimCurve
   defaultValue=235
   parameterName=BlinkCurve
  }
  n37 = Multiply(A=ScalarParameter() {defaultValue=0.5, parameterName=Blink}, B=@n36.B)
  n38 = LinearInterpolate(A=Round(None=Add(A=@n35, B=@n37)), Alpha=@n12)
  n39 = Subtract(A=Multiply(A=ComponentMask(None=@n5.RGB) {g=true, r=true}) {constB=2})
  n40 = ScalarParameter() {defaultValue=0.035, parameterName=AC_DistortionStrength}
  n41 = MaterialFunctionCall(Float2=Add(A=TextureCoordinate(), B=Multiply(A=@n39, B=@n40))) {materialFunction=&BreakOutFloat2Components}
  n42 = ScalarParameter() {parameterName=AC_LaughStretch}
  n43 = ScalarParameter() {defaultValue=0.05, parameterName=StretchAMtTop}
  n44 = MaterialFunctionCall(X=@n41.R, Y=Add(A=@n41.G, B=Multiply(A=@n42, B=@n43))) {materialFunction=&MakeFloat2}
  n45 = ScalarParameter() {defaultValue=-2, parameterName=FaceOffset}
  n46 = ScalarParameter() {defaultValue=1, parameterName=FaceScale}
  n47 = TextureSampleParameter2D(UVs=MaterialFunctionCall(UVs=BumpOffset(Coordinate=@n44.Result, Height=@n45) {referencePlane=0}, Texture Scale=@n46) {materialFunction=&ScaleUVsByCenter}.UVs) {
   parameterName=FaceFX_Mask
   samplerType=SAMPLERTYPE_Masks
   texture=/Game/Accessories/FORT_Backpacks/Backpack_M_MED_Accumulate/Textures/T_MED_Accumulate_Backpack_FX
  }
  n48 = ChannelMaskParameter(None=@n47.RGB) {parameterName=DefaulFaceChannel}
  n49 = ScalarParameter() {defaultValue=0.1, parameterName=StretchAMtBottom}
  n50 = Add(A=@n41.G, B=Multiply(A=Multiply(A=@n42, B=@n49)) {constB=-1})
  n51 = MaterialFunctionCall(
   UVs=BumpOffset(
    Coordinate=MaterialFunctionCall(X=@n41.R, Y=@n50) {materialFunction=&MakeFloat2}.Result
    Height=@n45
   ) {referencePlane=0}
   Texture Scale=@n46
  ) {materialFunction=&ScaleUVsByCenter}
  n52 = TextureSampleParameter2D(UVs=@n51.UVs) {
   parameterName=FaceFX_Mask
   samplerType=SAMPLERTYPE_Masks
   texture=&T_MED_Accumulate_Backpack_FX
  }
  n53 = Add(A=@n47.G, B=@n52.B)
  n54 = Multiply(A=LinearInterpolate(A=Multiply(A=@n38, B=@n48.R), B=@n53, Alpha=@n12), B=@n51.0-1 mask)
  n55 = CurveAtlasRowParameter(CurveTime=@n54) {
   atlas=&AthenaCurveAtlas01
   curve=&AccumulateBlackColorCurve
   defaultValue=276
   parameterName=FaceColorCurve
  }
  n56 = LinearInterpolate(A=LinearInterpolate(B=Multiply(A=@n33, B=@n34), Alpha=@n12) {constB=0}, B=@n55, Alpha=@n53)
  n57 = Add(A=@n32, B=@n54)
  n58 = VectorParameter() {parameterName=Color_02, defaultValue=<rgba 0 0.4993 0.535 1>}
  n59 = VectorParameter() {parameterName=Color_01, defaultValue=<rgba 1 1 1 1>}
  n60 = Power(Base=Fresnel(ExponentIn=ScalarParameter() {defaultValue=1, parameterName=Exponent}))
  n61 = Constant()
  n62 = SetMaterialAttributes!(
   MaterialAttributes=GetMaterialAttributes(None=@n9.Result).MaterialAttributes
   Emissive Color=StaticSwitchParameter(True=LinearInterpolate(A=@n11, B=@n56, Alpha=@n57), False=LinearInterpolate(A=@n58.RGB, B=@n59.RGB, Alpha=@n60)) {parameterName=IsBasePart}
   Base Color=@n61
   Metallic=@n61
   Specular=@n61
   Roughness=Constant() {r=1}
   Normal=Constant3Vector() {constant=<rgba 0 0 1 0>}
  ) {attributeSetTypes [
    #1
    #2
    #3
    #4
    #5
    #6
   ]}
  MaterialFunctionCall(AO=@n9.AO, Distance Blend=@n9.AO, Fresnel=@n9.AO, Gmap=@n9.AO, Pre Skinned Local Position=@n9.AO, Material=@n62) {materialFunction=&BaseFN_CharacterFunction_PostFX}
  Multiply(A=Time()) {constB=0.25}
  n63 = MaterialFunctionCall(Float2=TextureCoordinate()) {materialFunction=&BreakOutFloat2Components}
  n64 = Subtract(A=Multiply(A=@n63.R) {constB=2})
  Multiply(A=Power(Base=Abs(None=@n64)) {constExponent=1.25}, B=Multiply(A=Divide(A=@n63.G)))
  n65 = Subtract(A=Multiply(A=TextureCoordinate()) {constB=2})
  n66 = Multiply(A=Distance(A=Add(A=@n65), B=Constant2Vector()), B=ScalarParameter() {defaultValue=4, parameterName=Rings})
  SphereMask(
   A=Frac(None=Add(A=@n66, B=@n17))
   B=ScalarParameter() {defaultValue=0.5, parameterName=SamplePos}
   Radius=ScalarParameter() {defaultValue=0.1, parameterName=Radius}
   Hardness=ScalarParameter() {parameterName=Hardness}
  )
  Time()
  VectorParameter() {parameterName=BG_Color, defaultValue=<rgba 0 0 0 1>}
 ]
}
